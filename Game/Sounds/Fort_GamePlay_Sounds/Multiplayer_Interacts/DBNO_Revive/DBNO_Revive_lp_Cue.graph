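# Sound cue for the looping sound of a DBNO (down but not out) revive interaction in multiplayer.
# A single looping wave player feeds a Modulator root left at its default settings.
# It uses a dedicated DBNO revive attenuation asset, plays in the in-game SFX sound class and carries a boosted volume multiplier.
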SoundCue /Game/Sounds/Fort_GamePlay_Sounds/Multiplayer_Interacts/DBNO_Revive/DBNO_Revive_lp_Cue
root=Modulator_0
nodes [
 {id=Modulator_0, children [WavePlayer_1]}
 {
  id=WavePlayer_1
  properties {bLooping=true, soundWaveAssetPtr=/FNGameplayCues/Sounds/Fort_GamePlay_Sounds/Multiplayer_Interacts/DBNO_Revive/wav/DBNO_Revive_lp_01}
 }
]
properties {
 attenuationSettings=/FNGameplayCues/Sounds/Fort_GamePlay_Sounds/Multiplayer_Interacts/DBNO_Revive/DBNO_Revive_Att
 maxDistance=768
 soundClassObject=&Fort_SFX_InGame
 volumeMultiplier=3
 assetUserData [&self:AudioAssetUserData_0]
}
